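# Previs Level Sequence for the NightNight Part 4 rocket test 1A1, flying a separated Geode rocket along a rail.
# A CameraRig_Rail animates its rail position from start to end while a hidden, rolling parent mesh, the rocket and an enter-FX emitter use attach tracks.
# The rocket spawns only for the flight window with passby audio disabled, and its Stage2 mesh scales to zero as the rail run ends.
# A static root mesh and a hugely stretched cylinder stay spawned throughout.
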LevelSequence /Figment_S10/Athena/Prototype/Blueprints/NightNight/Sequences/RocketTests_P4_1/NightNight_Animation_PREVIS_Part4_RocketTest1A1
bindings [
 Rocket_01_Sep {
  id=#1
  possessed_class=BP_Geode_NightNight_Separated_C
  tracks [
   {
    type=MovieSceneSpawnTrack
    sections [{
      type=MovieSceneSpawnSection
      channels [None {type=MovieSceneScriptingBoolChannel, default=true, keys [[0,false],[1019,true],[1468.8,false]]}]
     }]
   }
   {
    type=MovieScene3DTransformTrack
    sections [{
      type=MovieScene3DTransformSection
      end_frame=2642
      start_frame=1019
      channels [
       Location.X {type=MovieSceneScriptingDoubleChannel, default=0}
       Location.Y {type=MovieSceneScriptingDoubleChannel, default=0}
       Location.Z {type=MovieSceneScriptingDoubleChannel, default=0}
       Rotation.X {type=MovieSceneScriptingDoubleChannel, default=-44.6252}
       Rotation.Y {type=MovieSceneScriptingDoubleChannel, default=-90}
       Rotation.Z {type=MovieSceneScriptingDoubleChannel, default=44.6252}
      ]
     }]
   }
   {type=MovieScene3DAttachTrack, sections [{type=MovieScene3DAttachSection, end_frame=2637, start_frame=1019}]}
   {
    type=MovieSceneBoolTrack
    display_name=Disable Passby Audio
    sections [{type=MovieSceneBoolSection, channels [None {type=MovieSceneScriptingBoolChannel, default=true}]}]
   }
  ]
 }
 MissileSilo_Stage2_Main {
  id=#2
  parent_id=#1
  possessed_class=StaticMeshComponent
  tracks [{
    type=MovieScene3DTransformTrack
    sections [{
      type=MovieScene3DTransformSection
      channels [
       Scale.X {type=MovieSceneScriptingDoubleChannel, default=0, keys [[1302,1],[1319,0]]}
       Scale.Y {type=MovieSceneScriptingDoubleChannel, default=0, keys [[1302,1],[1319,0]]}
       Scale.Z {type=MovieSceneScriptingDoubleChannel, default=0, keys [[1302,1],[1319,0]]}
      ]
     }]
   }]
 }
 Rocket_01_Path {
  id=#3
  possessed_class=CameraRig_Rail
  tracks [
   {
    type=MovieSceneSpawnTrack
    sections [{type=MovieSceneSpawnSection, channels [None {type=MovieSceneScriptingBoolChannel, default=false, keys [[0,false],[1019,true]]}]}]
   }
   {
    type=MovieScene3DTransformTrack
    sections [{
      type=MovieScene3DTransformSection
      channels [
       Location.X {type=MovieSceneScriptingDoubleChannel, default=0.293}
       Location.Y {type=MovieSceneScriptingDoubleChannel, default=167.0801}
       Location.Z {type=MovieSceneScriptingDoubleChannel, default=87792.2188}
       Rotation.X {type=MovieSceneScriptingDoubleChannel, default=90}
       Rotation.Y {type=MovieSceneScriptingDoubleChannel, default=0}
       Rotation.Z {type=MovieSceneScriptingDoubleChannel, default=0}
       Scale.X {type=MovieSceneScriptingDoubleChannel, default=1}
       Scale.Y {type=MovieSceneScriptingDoubleChannel, default=1}
       Scale.Z {type=MovieSceneScriptingDoubleChannel, default=1}
      ]
     }]
   }
   {
    type=MovieSceneFloatTrack
    display_name=Current Position on Rail
    sections [{type=MovieSceneFloatSection, channels [None {type=MovieSceneScriptingFloatChannel, default=0, keys [[1023,0],[1319,1]]}]}]
   }
   {type=MovieScene3DAttachTrack, sections [{type=MovieScene3DAttachSection, end_frame=2642, start_frame=1019}]}
  ]
 }
 RocketRoot {
  id=#4
  possessed_class=FortStaticMeshActor
  tracks [
   {
    type=MovieScene3DTransformTrack
    sections [{
      type=MovieScene3DTransformSection
      channels [
       Location.X {type=MovieSceneScriptingDoubleChannel, default=25055.9961}
       Location.Y {type=MovieSceneScriptingDoubleChannel, default=1894.965}
       Location.Z {type=MovieSceneScriptingDoubleChannel, default=3724.0361}
       Rotation.X {type=MovieSceneScriptingDoubleChannel, default=662.0326}
       Rotation.Y {type=MovieSceneScriptingDoubleChannel, default=-31.5519}
       Rotation.Z {type=MovieSceneScriptingDoubleChannel, default=-498.141}
       Scale.X {type=MovieSceneScriptingDoubleChannel, default=1}
       Scale.Y {type=MovieSceneScriptingDoubleChannel, default=1}
       Scale.Z {type=MovieSceneScriptingDoubleChannel, default=1}
      ]
     }]
   }
   {type=MovieSceneSpawnTrack, sections [{type=MovieSceneSpawnSection, channels [None {type=MovieSceneScriptingBoolChannel, default=true}]}]}
  ]
 }
 Rocket_01_Enter_FX {
  id=#5
  possessed_class=Emitter
  tracks [
   {
    type=MovieSceneSpawnTrack
    sections [{type=MovieSceneSpawnSection, channels [None {type=MovieSceneScriptingBoolChannel, default=false, keys [[0,false],[1023,true]]}]}]
   }
   {
    type=MovieScene3DTransformTrack
    sections [{
      type=MovieScene3DTransformSection
      channels [
       Location.X {type=MovieSceneScriptingDoubleChannel, default=31.6797}
       Location.Y {type=MovieSceneScriptingDoubleChannel, default=-162.7793}
       Location.Z {type=MovieSceneScriptingDoubleChannel, default=203226.3438}
       Rotation.X {type=MovieSceneScriptingDoubleChannel, default=139.5328}
       Rotation.Y {type=MovieSceneScriptingDoubleChannel, default=-83.3572}
       Rotation.Z {type=MovieSceneScriptingDoubleChannel, default=178.2421}
       Scale.X {type=MovieSceneScriptingDoubleChannel, default=1}
       Scale.Y {type=MovieSceneScriptingDoubleChannel, default=1}
       Scale.Z {type=MovieSceneScriptingDoubleChannel, default=1}
      ]
     }]
   }
   {
    type=MovieSceneParticleTrack
    display_name=Particle System
    sections [{type=MovieSceneParticleSection, channels [None {type=MovieSceneScriptingParticleChannel, const ["<ParticleKey.TRIGGER: 2>", 1]}]}]
   }
   {type=MovieScene3DAttachTrack, sections [{type=MovieScene3DAttachSection, end_frame=2642, start_frame=1019}]}
  ]
 }
 Cylinder {
  id=#6
  possessed_class=FortStaticMeshActor
  tracks [
   {
    type=MovieScene3DTransformTrack
    sections [{
      type=MovieScene3DTransformSection
      channels [
       Location.X {type=MovieSceneScriptingDoubleChannel, default=0}
       Location.Y {type=MovieSceneScriptingDoubleChannel, default=0}
       Location.Z {type=MovieSceneScriptingDoubleChannel, default=0}
       Rotation.X {type=MovieSceneScriptingDoubleChannel, default=0}
       Rotation.Y {type=MovieSceneScriptingDoubleChannel, default=0}
       Rotation.Z {type=MovieSceneScriptingDoubleChannel, default=0}
       Scale.X {type=MovieSceneScriptingDoubleChannel, default=0.7}
       Scale.Y {type=MovieSceneScriptingDoubleChannel, default=0.7}
       Scale.Z {type=MovieSceneScriptingDoubleChannel, default=99999}
      ]
     }]
   }
   {type=MovieSceneSpawnTrack, sections [{type=MovieSceneSpawnSection, channels [None {type=MovieSceneScriptingBoolChannel, default=true}]}]}
   {type=MovieScene3DAttachTrack, sections [{type=MovieScene3DAttachSection, end_frame=2642, start_frame=0}]}
  ]
 }
 Rocket_01_Parent {
  id=#7
  possessed_class=FortStaticMeshActor
  tracks [
   {
    type=MovieScene3DTransformTrack
    sections [
     {
      type=MovieScene3DTransformSection
      channels [
       Location.X {type=MovieSceneScriptingDoubleChannel, default=0, const [0,1]}
       Location.Y {type=MovieSceneScriptingDoubleChannel, default=0, const [0,1]}
       Location.Z {type=MovieSceneScriptingDoubleChannel, default=0, const [0,1]}
       Rotation.X {type=MovieSceneScriptingDoubleChannel, default=0, keys [[1019,0],[1202,119.0061],[1319,-69.2726]]}
       Rotation.Y {type=MovieSceneScriptingDoubleChannel, default=0, keys [[1019,0],[1202,0.0013],[1319,0.0019]]}
       Rotation.Z {type=MovieSceneScriptingDoubleChannel, default=0, keys [[1019,0],[1202,0.0009],[1319,0.0007]]}
       Scale.X {type=MovieSceneScriptingDoubleChannel, default=1, const [1,1]}
       Scale.Y {type=MovieSceneScriptingDoubleChannel, default=1, const [1,1]}
       Scale.Z {type=MovieSceneScriptingDoubleChannel, default=1, const [1,1]}
      ]
     }
     {
      type=MovieScene3DTransformSection
      end_frame=2642
      start_frame=571
      channels [
       Location.X {type=MovieSceneScriptingDoubleChannel, default=0}
       Location.Y {type=MovieSceneScriptingDoubleChannel, default=0}
       Location.Z {type=MovieSceneScriptingDoubleChannel, default=0}
       Rotation.X {type=MovieSceneScriptingDoubleChannel, default=0}
       Rotation.Y {type=MovieSceneScriptingDoubleChannel, default=0}
       Rotation.Z {type=MovieSceneScriptingDoubleChannel, default=0}
      ]
     }
    ]
   }
   {type=MovieSceneSpawnTrack, sections [{type=MovieSceneSpawnSection, channels [None {type=MovieSceneScriptingBoolChannel, default=true}]}]}
   {type=MovieScene3DAttachTrack, sections [{type=MovieScene3DAttachSection, end_frame=2642, start_frame=1019}]}
   {
    type=MovieSceneVisibilityTrack
    sections [{type=MovieSceneVisibilitySection, channels [None {type=MovieSceneScriptingBoolChannel, default=false}]}]
   }
  ]
 }
]
display_rate {denominator=1, numerator=60}
tick_resolution {denominator=1, numerator=24000}
tracks []
